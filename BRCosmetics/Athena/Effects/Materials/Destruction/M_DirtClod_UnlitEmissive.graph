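# Unlit masked material for Niagara sprite dirt clod particles, built on the dirt clod cluster texture.
# The texture's alpha serves as a bump offset height to parallax-shift the UVs of a second sample of the same texture.
# That sample's color is multiplied by particle color.
# Its inverted alpha is subtracted from particle alpha, so lowering particle alpha erodes the clod shapes rather than fading them.
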Material /BRCosmetics/Athena/Effects/Materials/Destruction/M_DirtClod_UnlitEmissive
graph {
 properties {bUsedWithNiagaraSprites=true, blendMode=BLEND_Masked, pixelDepthOffsetMode=PDOM_Legacy, shadingModel=MSM_Unlit, usedShadingModels=Unlit}
 nodes [
  n1 = ParticleColor()
  n2 = TextureSample(UVs=BumpOffset(Height=TextureSample() {texture=/Game/Effects/Fort_Effects/Textures/Dirt/T-DirtClodCluster_D}.A)) {texture=&T-DirtClodCluster_D}
  Multiply(A=@n1.RGB, B=@n2.RGB)
  Subtract(A=@n1.A, B=OneMinus(None=@n2.A))
 ]
}
